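# Masked unlit material for Niagara sprite and mesh particles, drawing the Teriyaki Fish Toon backpack's black surf bubble.
# A camera-space view-normal cross product gives matcap-style coordinates for sphere masks, forming a white rim and a highlight that fades in with camera distance.
# Cellular noise, offset by particle Random, meets a Reveal-driven vertical threshold for dissolving.
# A lagged second threshold adds a white edge band, while HeightClip also cuts the mask.
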Material /BRCosmetics/Accessories/FORT_Backpacks/Backpack_M_MED_Teriyaki_Fish_Toon/FX/M_TeriyakiFishToon_BlackSurfBubble
graph {
 properties {
  bEnableResponsiveAA=true
  bUsedWithNiagaraMeshParticles=true
  bUsedWithNiagaraSprites=true
  blendMode=BLEND_Masked
  pixelDepthOffsetMode=PDOM_Legacy
  shadingModel=MSM_Unlit
  usedShadingModels=Unlit
 }
 nodes [
  n1 = WorldPosition() {worldPositionShaderOffset=WPT_CameraRelative}
  n2 = CrossProduct(
   A=Normalize(VectorInput=Transform(None=@n1.XYZ) {transformSourceType=TRANSFORMSOURCE_World, transformType=TRANSFORM_Camera})
   B=Transform(None=VertexNormalWS()) {transformSourceType=TRANSFORMSOURCE_World, transformType=TRANSFORM_Camera}
  )
  n3 = AppendVector(A=Subtract(B=ComponentMask(None=@n2) {g=true}) {constA=0}, B=ComponentMask(None=@n2) {r=true})
  n4 = Constant()
  n5 = ComponentMask(None=VertexInterpolator(VS=AppendVector(A=@n3, B=@n4)).PS) {g=true, r=true}
  n6 = Constant2Vector()
  n7 = CameraPositionWS()
  n8 = WorldPosition()
  n9 = MaterialFunctionCall(
   Input=MaterialFunctionCall(Vector 3=Subtract(A=@n7, B=@n8.XYZ)) {materialFunction=/Engine/Functions/Engine_MaterialFunctions02/Utility/VectorLength}.V3 Length
   Input Low=ScalarParameter() {defaultValue=1000, parameterName=MinDIST}
   Input High=ScalarParameter() {defaultValue=5000, parameterName=MaxDIST}
   Target Low=Constant()
   Target High=Constant() {r=1}
  ) {materialFunction=&RemapValueRange}
  n10 = SphereMask(A=@n5, B=Constant2Vector() {g=-0.35, r=-0.35}) {attenuationRadius=0.2, hardnessPercent=90}
  n11 = Add(
   A=OneMinus(None=SphereMask(A=@n5, B=@n6) {attenuationRadius=0.85, hardnessPercent=90})
   B=Multiply(A=VertexInterpolator(VS=@n9.Result).PS, B=Saturate(None=@n10))
  )
  n12 = Constant3Vector() {desc=OutlineColor, constant=<rgba 1 1 1 0>}
  n13 = TextureCoordinate()
  n14 = DynamicParameter() {defaultValue=<rgba 1 -1 -1 1>, paramNames [Random,SpherePosY,Reveal,Param4]}
  n15 = Add(A=OneMinus(None=ComponentMask(None=@n13) {g=true}), B=@n14.Reveal)
  n16 = TextureCoordinate() {uTiling=0.4, vTiling=0.4}
  n17 = TextureSample(UVs=Panner(Coordinate=Add(A=@n16, B=@n14.Random))) {samplerType=SAMPLERTYPE_LinearGrayscale, texture=/BRCosmetics/Effects/Fort_Effects/BluGlow/Violentium/Cellular_basecolor}
  n18 = ScalarParameter() {parameterName=lag}
  n19 = SmoothStep(Min=@n15, Max=Add(A=@n15, B=ScalarParameter() {parameterName=w}), Value=AppendVector(A=@n17.G, B=Subtract(A=@n17.G, B=@n18)))
  n20 = ComponentMask(None=@n19) {r=true}
  n21 = ComponentMask(None=@n19) {g=true}
  n22 = Saturate(None=Subtract(A=@n20, B=@n21))
  MaterialFunctionCall(Emissive=Clamp(None=LinearInterpolate(A=@n11, B=@n12, Alpha=@n22))) {materialFunction=&MF_FX_EmissiveAdjustments}
  n23 = TextureCoordinate()
  Multiply(
   A=Step(Y=ComponentMask(None=@n23) {g=true}, X=ScalarParameter() {parameterName=HeightClip}) {constX=0.5}
   B=Saturate(None=Add(A=@n20, B=@n21))
  )
 ]
}
