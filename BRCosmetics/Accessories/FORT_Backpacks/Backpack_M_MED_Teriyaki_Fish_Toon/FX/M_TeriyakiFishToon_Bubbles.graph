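# Unlit masked Niagara sprite material for the bubble effect on the Teriyaki Fish Toon back bling.
# A SphereMask builds a circular shape in UV space, its radius wobbled by panning soft noise remapped around zero and scaled by the DistortionWobble parameter.
# Particle color RGB feeds a shared emissive-adjustment function, so color comes from the particles rather than a texture.
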Material /BRCosmetics/Accessories/FORT_Backpacks/Backpack_M_MED_Teriyaki_Fish_Toon/FX/M_TeriyakiFishToon_Bubbles
graph {
 properties {bUsedWithNiagaraSprites=true, blendMode=BLEND_Masked, pixelDepthOffsetMode=PDOM_Legacy, shadingModel=MSM_Unlit, usedShadingModels=Unlit}
 nodes [
  MaterialFunctionCall(Emissive=ParticleColor().RGB) {materialFunction=&MF_FX_EmissiveAdjustments}
  n1 = Panner(Coordinate=TextureCoordinate() {uTiling=0.2, vTiling=0.2}) {speedX=-0.1238, speedY=0.5}
  n2 = Subtract(A=Multiply(A=TextureSample(UVs=@n1) {texture=/Game/Characters/Player/Male/Medium/Bodies/M_MED_Viper/Textures/Glitch/T_SoftNoise}.R) {constB=2})
  n3 = ScalarParameter() {defaultValue=0.125, parameterName=DistortionWobble}
  SphereMask(A=TextureCoordinate(), B=Constant() {r=0.5}, Radius=Add(A=Constant() {r=0.4}, B=Multiply(A=@n2, B=@n3) {constB=0.125})) {attenuationRadius=0.5}
 ]
}
