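# Unlit masked Niagara sprite material for a toon splash effect on the Teriyaki Fish Toon back bling.
# The splash texture's green channel is thresholded by a SmoothStep driven by the particle's Erode dynamic parameter, giving animated dissolve.
# A second threshold, offset by OutlineW, forms an outline region, and both masks are summed and saturated.
# The first mask blends white to Color (default black), feeding an emissive adjustment function.
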Material /BRCosmetics/Accessories/FORT_Backpacks/Backpack_M_MED_Teriyaki_Fish_Toon/FX/M_TeriyakiFishToon_SplashElemn
graph {
 properties {bUsedWithNiagaraSprites=true, blendMode=BLEND_Masked, pixelDepthOffsetMode=PDOM_Legacy, shadingModel=MSM_Unlit, usedShadingModels=Unlit}
 nodes [
  n1 = DynamicParameter() {paramNames [Erode,Param2,Param3,Param4]}
  n2 = TextureSample() {
   samplerSource=SSM_Clamp_WorldGroupSettings
   samplerType=SAMPLERTYPE_LinearColor
   texture=&/T_FX_TeriyakiToon_Splash
  }
  n3 = ScalarParameter() {parameterName=OutlineW}
  n4 = SmoothStep(Min=@n1.Erode, Max=Add(A=@n1.Erode, B=ScalarParameter() {parameterName=w}), Value=AppendVector(A=@n2.G, B=Add(A=@n2.G, B=@n3)))
  n5 = ComponentMask(None=@n4) {r=true}
  MaterialFunctionCall(Emissive=LinearInterpolate(B=VectorParameter() {parameterName=Color, defaultValue=<rgba 0 0 0 1>}.RGB, Alpha=@n5) {constA=1}) {materialFunction=&MF_FX_EmissiveAdjustments}
  Saturate(None=Add(A=@n5, B=ComponentMask(None=@n4) {g=true}))
  ScalarParameter() {parameterName=Edge}
 ]
}
