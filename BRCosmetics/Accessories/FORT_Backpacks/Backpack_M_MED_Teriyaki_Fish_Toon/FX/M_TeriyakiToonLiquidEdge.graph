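# Unlit masked Niagara particle material for the liquid edge effect on the toon Teriyaki Fish backpack, with particle color fed through an emissive adjustment function.
# A panning, inverted cellular texture is smoothstepped against a sphere-mask threshold for hard, shaped breakup, timed by game or particle time via static switch.
# A vertex offset combines a view-space shift and sine ripple along normals, weighted by inverted V and scaled by the IsDivingFast dynamic parameter.
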Material /BRCosmetics/Accessories/FORT_Backpacks/Backpack_M_MED_Teriyaki_Fish_Toon/FX/M_TeriyakiToonLiquidEdge
graph {
 properties {
  bUsedWithNiagaraMeshParticles=true
  bUsedWithNiagaraSprites=true
  blendMode=BLEND_Masked
  pixelDepthOffsetMode=PDOM_Legacy
  shadingModel=MSM_Unlit
  usedShadingModels=Unlit
 }
 nodes [
  Constant3Vector() {constant=<rgba 1 0 0 0>}
  MaterialFunctionCall(Emissive=ParticleColor().RGB) {materialFunction=&MF_FX_EmissiveAdjustments}
  n1 = TextureCoordinate()
  n2 = Multiply(
   A=OneMinus(None=ComponentMask(None=@n1) {g=true})
   B=Transform(None=VectorParameter() {parameterName=Shift, defaultValue=<rgba 0 0 0 1>}.RGB) {transformSourceType=TRANSFORMSOURCE_View}
  )
  n3 = TextureCoordinate()
  n4 = TextureCoordinate() {vTiling=3}
  n5 = Sine(None=Add(A=ComponentMask(None=@n4) {g=true}, B=Time()))
  n6 = ScalarParameter() {parameterName=Push}
  n7 = Multiply(A=OneMinus(None=ComponentMask(None=@n3) {g=true}), B=Multiply(A=VertexNormalWS(), B=Multiply(A=@n5, B=@n6)))
  Multiply(
   A=Multiply(A=Add(A=@n2, B=@n7), B=DynamicParameter() {paramNames [IsDivingFast,Time,Param3,Param4]}.IsDivingFast)
   B=ScalarParameter() {parameterName=DivingVS_Strength}
  )
  n8 = TextureCoordinate()
  n9 = Constant2Vector() {g=1, r=0.5}
  n10 = ScalarParameter() {parameterName=r}
  n11 = Add(
   A=ScalarParameter() {parameterName=step_pos}
   B=Power(Base=SphereMask(A=@n8, B=@n9, Radius=@n10) {attenuationRadius=0.5, hardnessPercent=0}) {constExponent=1}
  )
  n12 = VectorParameter() {parameterName=Tex_Scale, defaultValue=<rgba 0 0 0 1>}
  n13 = Panner(
   Coordinate=Multiply(A=TextureCoordinate() {uTiling=0.4, vTiling=0.4}, B=ComponentMask(None=@n12.RGB) {g=true, r=true})
   Time=StaticSwitchParameter(True=DynamicParameter() {paramNames [IsDivingFast,Time,Param3,Param4]}.Time, False=Time()) {parameterName=TimeIsParticleDynamic}
  ) {speedY=1}
  SmoothStep(
   Min=@n11
   Max=Add(A=@n11, B=ScalarParameter() {parameterName=step_w})
   Value=OneMinus(None=TextureSample(UVs=@n13) {samplerType=SAMPLERTYPE_LinearGrayscale, texture=/BRCosmetics/Effects/Fort_Effects/BluGlow/Violentium/Cellular_basecolor}.G)
  )
 ]
}
