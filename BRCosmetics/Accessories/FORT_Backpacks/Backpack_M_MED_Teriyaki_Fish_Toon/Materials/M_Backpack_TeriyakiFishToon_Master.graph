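# Master toon material for the Teriyaki Fish Toon backpack, drawing its color through emissive over black base color.
# Base and outline colors blend by a contrasted Fresnel masked by pre-skinned height, with a vertex-red mask adding white, then pass through time-of-day and LUT functions.
# Two-sided sign times FlipNormals masks faces by orientation, except in shadows.
# Distance-based line thickness and sun/moon specular highlight chains end without visible connections.
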Material /BRCosmetics/Accessories/FORT_Backpacks/Backpack_M_MED_Teriyaki_Fish_Toon/Materials/M_Backpack_TeriyakiFishToon_Master
graph {
 properties {
  bDisableDepthTest=true
  bEnableResponsiveAA=true
  bUseMaterialAttributes=true
  bUsedWithMeshDeformer=true
  bUsedWithSkeletalMesh=true
  ditherOpacityMask=true
  pixelDepthOffsetMode=PDOM_Legacy
  shadingModel=MSM_SubsurfaceProfile
  translucencyDirectionalLightingIntensity=5
  translucencyLightingMode=TLM_Surface
  usedShadingModels=Subsurface Profile
 }
 nodes [
  n1 = ScalarParameter() {defaultValue=100, group=NeonDepth, parameterName=DF_FarDistance}
  n2 = Subtract(A=Distance(A=WorldPosition().XYZ, B=CameraPositionWS()), B=@n1)
  n3 = Subtract(A=ScalarParameter() {defaultValue=500, group=NeonDepth, parameterName=DF_NearDistance}, B=@n1)
  LinearInterpolate(
   A=ScalarParameter() {defaultValue=50, group=NeonDepth, parameterName=DF_LineThickness_Near}
   B=ScalarParameter() {defaultValue=1, group=NeonDepth, parameterName=DF_LineThickness_Far}
   Alpha=Saturate(None=Divide(A=@n2, B=@n3))
  )
  n4 = TwoSidedSign()
  n5 = ScalarParameter() {defaultValue=-1, group=0SlurpMonsterSummer, parameterName=FlipNormals}
  n6 = VectorParameter() {parameterName=Base Color, defaultValue=<rgba 0.6042 0.6042 0.6042 0>}
  n7 = VectorParameter() {parameterName=Outline Color}
  n8 = PreSkinnedPosition()
  n9 = PreSkinnedLocalBounds()
  n10 = ComponentMask(None=Divide(A=VertexInterpolator(VS=@n8).PS, B=Subtract(A=@n9.Max, B=@n9.Max))) {b=true}
  n11 = ScalarParameter() {defaultValue=0.1, parameterName=Fresnel Exponent_Outline}
  n12 = Multiply(
   A=Saturate(None=ConstantBiasScale(None=@n10) {bias=0.8, scale=3})
   B=MaterialFunctionCall(In=Fresnel(ExponentIn=@n11), Contrast=Constant() {r=50}) {materialFunction=&CheapContrast}.Result
  )
  n13 = SphereMask(A=VertexColor().R, B=Constant() {r=0.5}, Radius=Constant() {r=0.01})
  n14 = Multiply(A=LinearInterpolate(A=LinearInterpolate(A=@n6.RGB, B=@n7.RGB, Alpha=@n12), B=Constant() {r=1}, Alpha=Ceil(None=@n13))) {constB=0.6}
  n15 = Constant()
  n16 = SetMaterialAttributes!(
   MaterialAttributes
   Base Color=Constant()
   Opacity Mask=ShadowReplace(Default=Multiply(A=@n4, B=@n5), Shadow=Constant() {r=1})
   Emissive Color=MaterialFunctionCall!(
    In=MaterialFunctionCall!(Color=@n14) {materialFunction=/Game/Packages/Fortress_SharedMaterials/Anime_Master_Material/MF_TODColorCurve}.Result
    LUT
    LUT Multiplier
   ) {materialFunction=/Game/Packages/Fortress_SharedMaterials/Anime_Master_Material/MF_ApplyLUT}.Result
   Roughness
   Subsurface Color=ScalarParameter() {parameterName=Subsurface Color}
   Metallic=@n15
   Specular=@n15
  ) {attributeSetTypes [
    #1
    #2
    #3
    #4
    #5
    #6
    #7
   ]}
  MaterialFunctionCall(In=@n16) {materialFunction=&MF_TechArtOverrides_accessories}
  Multiply(B=@n11)
  n17 = CollectionParameter() {
   collection=&FortniteMaterialParameters
   parameterName=SunAndMoonModelDirectionalVector
  }
  n18 = Clamp(None=DotProduct(A=ComponentMask(None=@n17) {b=true, g=true, r=true}, B=ReflectionVectorWS()))
  n19 = ScalarParameter() {defaultValue=70, parameterName=Spec Highlight Power}
  Multiply(
   A=SphereMask(
    A=Power(Base=@n18, Exp=@n19) {constExponent=20}
    B=Constant() {r=1}
    Radius=ScalarParameter() {defaultValue=1, parameterName=Spec Highlight Radius}
    Hardness=ScalarParameter() {defaultValue=1, parameterName=Spec Highlight Hardness}
   )
   B=ScalarParameter() {defaultValue=3, parameterName=Spec Intense}
  )
 ]
}
