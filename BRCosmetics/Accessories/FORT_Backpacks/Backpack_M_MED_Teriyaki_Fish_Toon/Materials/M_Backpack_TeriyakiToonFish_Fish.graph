Material /BRCosmetics/Accessories/FORT_Backpacks/Backpack_M_MED_Teriyaki_Fish_Toon/Materials/M_Backpack_TeriyakiToonFish_Fish
graph {
 properties {
  bAutomaticallySetUsageInEditor=false
  bUseMaterialAttributes=true
  bUsedWithMeshDeformer=true
  bUsedWithSkeletalMesh=true
  pixelDepthOffsetMode=PDOM_Legacy
  shadingModel=MSM_Cloth
  usedShadingModels=Cloth
 }
 nodes [
  n1 = Constant()
  n2 = Constant() {r=0.9}
  n3 = VertexColor()
  n4 = ScalarParameter() {defaultValue=0.6, parameterName="Vertex for Custom Color "}
  n5 = Constant() {r=0.01}
  n6 = LinearInterpolate(
   A=@n2
   B=LinearInterpolate(A=@n2, B=VectorParameter() {parameterName=CustomColor, defaultValue=<rgba 0.9 0.9 0.9 0>}.RGB, Alpha=SphereMask(A=@n3.G, B=@n4, Radius=@n5)) {constA=11111}
   Alpha=ScalarParameter() {parameterName=Use Custom Color}
  )
  n7 = CollectionParameter() {
   collection=&FortniteMaterialParameters
   parameterName=SunlightDirectionalVector
  }
  n8 = ComponentMask(None=@n7) {b=true, g=true, r=true}
  n9 = PixelNormalWS()
  n10 = Multiply(A=@n6, B=SmoothStep(Max=ScalarParameter() {defaultValue=0.1, parameterName=ShadingSmooth}, Value=DotProduct(A=@n8, B=@n9)) {constMax=0})
  n11 = ScalarParameter() {defaultValue=0.45, parameterName=ShadingDarkenAmount}
  n12 = Constant() {r=4}
  n13 = PixelNormalWS()
  n14 = LinearInterpolate(
   A=LinearInterpolate(A=@n6, B=LinearInterpolate(A=@n6, B=@n10, Alpha=@n11), Alpha=ScalarParameter() {parameterName=EnableShading})
   Alpha=MaterialFunctionCall(In=Fresnel(ExponentIn=@n12, Normal=@n13), Contrast=Constant() {r=50}) {materialFunction=&CheapContrast}.Result
  ) {constA=1, constB=0}
  n15 = ScalarParameter() {defaultValue=0.75, parameterName=Brightness}
  n16 = TextureSampleParameter2D() {
   parameterName=DistanceField_InkLines
   samplerType=SAMPLERTYPE_Grayscale
   texture=/Game/Global/Textures/Default/Blanks/T_BlackLinearGrayscale_sRGB
  }
  n17 = TextureSampleParameter2D() {
   parameterName=DistanceField_InkLines_Mobile
   samplerType=SAMPLERTYPE_Grayscale
   texture=&T_BlackLinearGrayscale_sRGB
  }
  n18 = ChannelMaskParameter(None=ShadingPathSwitch(Default=@n16.RGB, Deferred=@n16.RGB, Mobile=@n17.RGB) {desc=Migrated from FeatureLevelSwitch}) {parameterName=DistanceField_Channel}
  n19 = Multiply(
   A=Power(Base=@n18.R, Exp=ScalarParameter() {defaultValue=9, parameterName=DF_BlurPow}) {constExponent=3}
   B=ScalarParameter() {defaultValue=1, parameterName=DF_BlurMult}
  ) {constB=0.5}
  n20 = ScalarParameter() {defaultValue=0.05, group=NeonDepth, parameterName=DF_LineThickness_Near}
  n21 = ShadingPathSwitch(Default=@n20, Deferred=@n20, Mobile=ScalarParameter() {defaultValue=0.05, group=NeonDepth, parameterName=DF_LineThickness_Near_Mobile}) {desc=Migrated from FeatureLevelSwitch}
  n22 = ScalarParameter() {defaultValue=0.15, group=NeonDepth, parameterName=DF_LineThickness_Far}
  n23 = ShadingPathSwitch(Default=@n22, Deferred=@n22, Mobile=ScalarParameter() {defaultValue=0.15, group=NeonDepth, parameterName=DF_LineThickness_Far_Mobile}) {desc=Migrated from FeatureLevelSwitch}
  n24 = Distance(A=WorldPosition().XYZ, B=CameraPositionWS())
  n25 = ScalarParameter() {defaultValue=100, group=NeonDepth, parameterName=DF_FarDistance}
  n26 = ScalarParameter() {defaultValue=500, group=NeonDepth, parameterName=DF_NearDistance}
  n27 = Saturate(None=Divide(A=Subtract(A=@n24, B=@n25), B=Subtract(A=@n26, B=@n25)))
  n28 = Saturate(None=SphereMask(
    A=@n18.R
    B=Constant() {r=1}
    Radius=LinearInterpolate(A=@n21, B=@n23, Alpha=@n27)
    Hardness=ScalarParameter() {defaultValue=0.8, parameterName=DF_LineHardness}
   ) {attenuationRadius=0.1})
  n29 = LinearInterpolate(A=Constant3Vector(), B=Constant3Vector() {constant=<rgba 1 1 1 0>}, Alpha=Multiply(A=OneMinus(None=@n19), B=OneMinus(None=@n28)))
  n30 = MaterialFunctionCall!(Color=Multiply(A=Multiply(A=@n14, B=@n15) {constB=0.75}, B=NamedRerouteDeclaration(None=@n29) {name=DF Line Preview, nodeColor=<rgba 0 1 0.0625 1>})) {materialFunction=/Game/Packages/Fortress_SharedMaterials/Anime_Master_Material/MF_TODColorCurve}
  n31 = TextureSampleParameter2D() {
   parameterName=Normals
   samplerType=SAMPLERTYPE_Normal
   texture=/BRCosmetics/Accessories/FORT_Backpacks/Backpack_M_MED_Teriyaki_Fish_Toon/Texture/T_M_MED_Teriyaki_Fish_Toon_Fish_N
  }
  n32 = SetMaterialAttributes!(
   MaterialAttributes
   Base Color=NamedRerouteDeclaration(None=@n1) {name=base, nodeColor=<rgba 0.7578 0 1 1>}
   Metallic=NamedRerouteDeclaration(None=@n1) {name=m, nodeColor=<rgba 1 0.7734 0 1>}
   Specular=NamedRerouteDeclaration(None=@n1) {name=s, nodeColor=<rgba 1 0.5859 0 1>}
   Roughness=NamedRerouteDeclaration(None=Constant() {r=1}) {name=r, nodeColor=<rgba 0.3828 0 1 1>}
   Emissive Color=NamedRerouteDeclaration(None=MaterialFunctionCall!(In=@n30.Result, LUT, LUT Multiplier) {materialFunction=/Game/Packages/Fortress_SharedMaterials/Anime_Master_Material/MF_ApplyLUT}.Result) {name=emissive, nodeColor=<rgba 0 1 0.3203 1>}
   Normal=NamedRerouteDeclaration(None=@n31.RGB) {name=n, nodeColor=<rgba 1 0.0234 0 1>}
   World Position Offset=NamedRerouteDeclaration(None=NamedRerouteUsage()) {name=wpo_1, nodeColor=<rgba 0.125 0 1 1>}
   Fuzz Color=NamedRerouteDeclaration(None=Constant() {r=0.001}) {name=ssc, nodeColor=<rgba 0.3359 0 1 1>}
   Ambient Occlusion=NamedRerouteDeclaration(None=NamedRerouteUsage()) {name=ambocc, nodeColor=<rgba 0 1 0.625 1>}
  ) {attributeSetTypes [
    #1
    #2
    #3
    #4
    #5
    #6
    #7
    #8
    #9
   ]}
  MaterialFunctionCall(In=@n32) {materialFunction=&MF_TechArtOverrides_accessories}
  n33 = Subtract(A=WorldPosition().XYZ, B=ObjectPositionWS())
  n34 = ChannelMaskParameter(None=VertexColor()) {parameterName=Vertex Hiding Channel}
  n35 = Max(
   A=MaterialFunctionCall!(
    Hide Element Toggle=ScalarParameter() {group=HideElements, parameterName=Hide Element 0_5}
    VertexColorChannel=@n34.R
    Vertex Value to hide=Constant() {r=0.05}
   ) {materialFunction=/Game/Packages/Fortress_SharedMaterials/Base_Material_Functions/VertexColorMask}.Result
   B=MaterialFunctionCall(
    Hide Element Toggle=ScalarParameter() {group=HideElements, parameterName=Hide Element 1_5}
    VertexColorChannel=@n34.R
    Vertex Value to hide=Constant() {r=0.15}
   ) {materialFunction=&VertexColorMask}.Result
  )
  n36 = MaterialFunctionCall(
   Hide Element Toggle=ScalarParameter() {group=HideElements, parameterName=Hide Element 2_5}
   VertexColorChannel=@n34.R
   Vertex Value to hide=Constant() {r=0.25}
  ) {materialFunction=&VertexColorMask}
  n37 = ScalarParameter() {group=HideElements, parameterName=Hide Element 3_5}
  n38 = Constant() {r=0.35}
  n39 = Max(
   A=Max(
    A=Max(A=@n35, B=@n36.Result)
    B=MaterialFunctionCall(Hide Element Toggle=@n37, VertexColorChannel=@n34.R, Vertex Value to hide=@n38) {materialFunction=&VertexColorMask}.Result
   )
   B=MaterialFunctionCall(
    Hide Element Toggle=ScalarParameter() {group=HideElements, parameterName=Hide Element 4_5}
    VertexColorChannel=@n34.R
    Vertex Value to hide=Constant() {r=0.45}
   ) {materialFunction=&VertexColorMask}.Result
  )
  n40 = MaterialFunctionCall(
   Hide Element Toggle=ScalarParameter() {group=HideElements, parameterName=Hide Element 5_5}
   VertexColorChannel=@n34.R
   Vertex Value to hide=Constant() {r=0.55}
  ) {materialFunction=&VertexColorMask}
  n41 = ScalarParameter() {group=HideElements, parameterName=Hide Element 6_5}
  n42 = Constant() {r=0.65}
  n43 = Max(
   A=Max(
    A=Max(A=@n39, B=@n40.Result)
    B=MaterialFunctionCall(Hide Element Toggle=@n41, VertexColorChannel=@n34.R, Vertex Value to hide=@n42) {materialFunction=&VertexColorMask}.Result
   )
   B=MaterialFunctionCall(
    Hide Element Toggle=ScalarParameter() {group=HideElements, parameterName=Hide Element 7_5}
    VertexColorChannel=@n34.R
    Vertex Value to hide=Constant() {r=0.75}
   ) {materialFunction=&VertexColorMask}.Result
  )
  n44 = MaterialFunctionCall(
   Hide Element Toggle=ScalarParameter() {group=HideElements, parameterName=Hide Element 8_5}
   VertexColorChannel=@n34.R
   Vertex Value to hide=Constant() {r=0.85}
  ) {materialFunction=&VertexColorMask}
  n45 = ScalarParameter() {group=HideElements, parameterName=Hide Element 9_5}
  n46 = Constant() {r=0.95}
  n47 = Max(
   A=MaterialFunctionCall(
    Hide Element Toggle=ScalarParameter() {group=HideElements, parameterName=Hide Element 01}
    VertexColorChannel=@n34.R
    Vertex Value to hide=Constant() {r=0.1}
   ) {materialFunction=&VertexColorMask}.Result
   B=MaterialFunctionCall(
    Hide Element Toggle=ScalarParameter() {group=HideElements, parameterName=Hide Element 02}
    VertexColorChannel=@n34.R
    Vertex Value to hide=Constant() {r=0.2}
   ) {materialFunction=&VertexColorMask}.Result
  )
  n48 = MaterialFunctionCall(
   Hide Element Toggle=ScalarParameter() {group=HideElements, parameterName=Hide Element 03}
   VertexColorChannel=@n34.R
   Vertex Value to hide=Constant() {r=0.3}
  ) {materialFunction=&VertexColorMask}
  n49 = ScalarParameter() {group=HideElements, parameterName=Hide Element 04}
  n50 = Constant() {r=0.4}
  n51 = Max(
   A=Max(
    A=Max(A=@n47, B=@n48.Result)
    B=MaterialFunctionCall(Hide Element Toggle=@n49, VertexColorChannel=@n34.R, Vertex Value to hide=@n50) {materialFunction=&VertexColorMask}.Result
   )
   B=MaterialFunctionCall(
    Hide Element Toggle=ScalarParameter() {group=HideElements, parameterName=Hide Element 05}
    VertexColorChannel=@n34.R
    Vertex Value to hide=Constant() {r=0.5}
   ) {materialFunction=&VertexColorMask}.Result
  )
  n52 = MaterialFunctionCall(
   Hide Element Toggle=ScalarParameter() {group=HideElements, parameterName=Hide Element 06}
   VertexColorChannel=@n34.R
   Vertex Value to hide=Constant() {r=0.6}
  ) {materialFunction=&VertexColorMask}
  n53 = ScalarParameter() {group=HideElements, parameterName=Hide Element 07}
  n54 = Constant() {r=0.7}
  n55 = Max(
   A=Max(
    A=Max(A=@n51, B=@n52.Result)
    B=MaterialFunctionCall(Hide Element Toggle=@n53, VertexColorChannel=@n34.R, Vertex Value to hide=@n54) {materialFunction=&VertexColorMask}.Result
   )
   B=MaterialFunctionCall(
    Hide Element Toggle=ScalarParameter() {group=HideElements, parameterName=Hide Element 08}
    VertexColorChannel=@n34.R
    Vertex Value to hide=Constant() {r=0.8}
   ) {materialFunction=&VertexColorMask}.Result
  )
  n56 = MaterialFunctionCall(
   Hide Element Toggle=ScalarParameter() {group=HideElements, parameterName=Hide Element 09}
   VertexColorChannel=@n34.R
   Vertex Value to hide=Constant() {r=0.9}
  ) {materialFunction=&VertexColorMask}
  n57 = ScalarParameter() {group=HideElements, parameterName=Hide Element 10}
  n58 = Constant() {r=1}
  n59 = Max(
   A=Max(
    A=Max(A=@n55, B=@n56.Result)
    B=MaterialFunctionCall(Hide Element Toggle=@n57, VertexColorChannel=@n34.R, Vertex Value to hide=@n58) {materialFunction=&VertexColorMask}.Result
   )
   B=Max(
    A=Max(A=@n43, B=@n44.Result)
    B=MaterialFunctionCall(Hide Element Toggle=@n45, VertexColorChannel=@n34.R, Vertex Value to hide=@n46) {materialFunction=&VertexColorMask}.Result
   )
  )
  n60 = VertexColor()
  n61 = Ceil(None=@n60)
  n62 = Max(A=Ceil(None=@n60.B), B=Ceil(None=@n60.G))
  n63 = OneMinus(None=Multiply(A=Max(A=@n61, B=@n62), B=ScalarParameter() {group=HideElements, parameterName=HideElementsOnlyConsidersRedChannel}))
  NamedRerouteDeclaration(None=Multiply(A=Multiply(A=@n33) {constB=-1}, B=Multiply(A=@n59, B=@n63))) {name=WPO, nodeColor=<rgba 0.2422 1 0 1>}
  ScalarParameter() {defaultValue=0.15, parameterName=lineThickness}
 ]
}
