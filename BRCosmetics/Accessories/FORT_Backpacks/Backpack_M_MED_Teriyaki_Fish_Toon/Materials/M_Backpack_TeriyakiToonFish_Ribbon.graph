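# Master material for the ribbon on the toon Teriyaki Fish backpack cosmetic, usable on skeletal meshes and mesh deformers.
# It fakes a flat toon look in Default Lit by zeroing base color, metallic and specular, with emissive from a constant gray run through time-of-day curve and LUT functions.
# World position offset sums a camera offset and a spline-thicken function with width and flatten parameters.
# The attributes pass through the shared accessory tech-art overrides function.
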Material /BRCosmetics/Accessories/FORT_Backpacks/Backpack_M_MED_Teriyaki_Fish_Toon/Materials/M_Backpack_TeriyakiToonFish_Ribbon
graph {
 properties {
  bUseMaterialAttributes=true
  bUsedWithMeshDeformer=true
  bUsedWithSkeletalMesh=true
  pixelDepthOffsetMode=PDOM_Legacy
  usedShadingModels=Default Lit
 }
 nodes [
  n1 = NamedRerouteDeclaration(None=Constant()) {name="basem netallic, spec", nodeColor=<rgba 0.1953 0 1 1>}
  n2 = MaterialFunctionCall!(Color=Constant() {r=0.75}) {materialFunction=/Game/Packages/Fortress_SharedMaterials/Anime_Master_Material/MF_TODColorCurve}
  n3 = MaterialFunctionCall(
   Offset Amount=ScalarParameter() {defaultValue=1, group=1, parameterName=CameraOffsetAmount}
   Clamp Padding=ScalarParameter() {group=1, parameterName=Clamp Padding}
  ) {materialFunction=/Engine/Functions/Engine_MaterialFunctions02/WorldPositionOffset/CameraOffset}
  n4 = MaterialFunctionCall!(
   widthTip=ScalarParameter() {defaultValue=1, group=1, parameterName=RibbonWidthTip}
   Flatten=ScalarParameter() {group=1, parameterName=RibbonFlatten}
   widthBase=ScalarParameter() {defaultValue=1, group=1, parameterName=RibbonWidthBase}
  ) {materialFunction=/Game/Characters/Player/Male/Medium/Bodies/M_MED_Neon/Materials/MF_splineThickenM}
  n5 = SetMaterialAttributes!(
   MaterialAttributes
   Base Color=@n1
   Metallic=@n1
   Specular=@n1
   Emissive Color=NamedRerouteDeclaration(None=MaterialFunctionCall!(In=@n2.Result, LUT, LUT Multiplier) {materialFunction=/Game/Packages/Fortress_SharedMaterials/Anime_Master_Material/MF_ApplyLUT}.Result) {name=emissive, nodeColor=<rgba 0 0.9531 1 1>}
   World Position Offset=NamedRerouteDeclaration(None=Add(A=@n3.Camera Offset, B=@n4.worldPositionOffset)) {name=wpo, nodeColor=<rgba 1 0 0.3047 1>}
  ) {attributeSetTypes [
    #1
    #2
    #3
    #4
    #5
   ]}
  MaterialFunctionCall(In=@n5) {materialFunction=&MF_TechArtOverrides_accessories}
 ]
}
