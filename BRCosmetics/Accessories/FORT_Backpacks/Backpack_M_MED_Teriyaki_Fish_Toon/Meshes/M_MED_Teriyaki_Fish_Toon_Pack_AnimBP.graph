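# Animation Blueprint for the Teriyaki Fish Toon backpack, playing weighted-random swim and idle-break animations.
# The pose runs through three spring-driven Anim Dynamics water bones, then an additive offset on the pack bone.
# Jump state sets jumpReactivity, the blend alpha for the first two water bones.
# A damage-triggered death and revival state machine, fed by health and shield decrease events, exists but is not connected to the output pose.
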AnimBlueprint /BRCosmetics/Accessories/FORT_Backpacks/Backpack_M_MED_Teriyaki_Fish_Toon/Meshes/M_MED_Teriyaki_Fish_Toon_Pack_AnimBP
target_skeleton=/Game/Accessories/FORT_Backpacks/Backpack_M_MED_Teriyaki_Fish_Toon/Meshes/M_MED_Teriyaki_Fish_Toon_Pack_Skeleton
graphs [
 AnimGraph {nodes [
   Input Pose()
   n1 = Random Sequence Player!() -> (Pose:Pose Link Structure) {
    is=title AnimGraphNode_RandomPlayer
    entries [
     {
      chanceToPlay=0.8
      maxLoopCount=0
      maxPlayRate=1
      minLoopCount=0
      minPlayRate=1
      sequence=/BRCosmetics/Accessories/FORT_Backpacks/Backpack_M_MED_Teriyaki_Fish_Toon/Animation/Teriyaki_Fish_Toon_Idle_Swim_BACKPACK
      blendIn {blendOption=Linear, blendTime=0.2, customCurve=null}
     }
     {
      chanceToPlay=0.2
      maxLoopCount=0
      maxPlayRate=1
      minLoopCount=0
      minPlayRate=1
      sequence=/BRCosmetics/Accessories/FORT_Backpacks/Backpack_M_MED_Teriyaki_Fish_Toon/Animation/Teriyaki_Fish_Toon_Idle_Break_BACKPACK
      blendIn {blendOption=Linear, blendTime=0.2, customCurve=null}
     }
    ]
   }
   n2 = Get jumpReactivity() -> (jumpReactivity:double)
   n3 = Anim Dynamics - Bone: dyn_waterDynamics_1(ComponentPose=Local To Component(LocalPose=@n1), Alpha=@n2, ExternalForce="0,0,0") {
    angularSpringConstant=500
    bAngularSpring=true
    bGravityOverrideInSimSpace=true
    gravityScale=0
    alphaBoolBlend {blendOption=Linear}
    alphaScaleBiasClamp {bInterpResult=true, interpSpeedDecreasing=20, interpSpeedIncreasing=20, scale=0.82}
    boundBone {boneName=dyn_waterDynamics_1}
    componentLinearAccScale=<0.9 0.9 0.9>
    physicsBodyDefinitions [{
      sphereCollisionRadius=0
      boundBone {boneName=dyn_waterDynamics_1}
      boxExtents=<5 5 5>
      constraintSetup {angularLimitsMax=<15 30 30>, angularLimitsMin=<-15 -2 -30>, angularTarget=<1 0 0>}
      localJointOffset=<0 -0.2 0.5>
     }]
   }
   n4 = Anim Dynamics - Bone: dyn_waterDynamics_2(ComponentPose=@n3, Alpha=@n2, ExternalForce="0,0,0") {
    angularSpringConstant=500
    bAngularSpring=true
    bGravityOverrideInSimSpace=true
    gravityScale=0
    alphaBoolBlend {blendOption=Linear}
    alphaScaleBiasClamp {bInterpResult=true, interpSpeedDecreasing=20, interpSpeedIncreasing=20, scale=0.82}
    boundBone {boneName=dyn_waterDynamics_2}
    componentLinearAccScale=<0.9 0.9 0.9>
    physicsBodyDefinitions [{
      sphereCollisionRadius=0
      boundBone {boneName=dyn_waterDynamics_2}
      boxExtents=<5 5 5>
      constraintSetup {angularLimitsMax=<15 40 30>, angularLimitsMin=<-15 -2 -30>, angularTarget=<1 0 0>}
      localJointOffset=<0 0.2 -0.5>
     }]
   }
   n5 = Anim Dynamics - Bone: dyn_waterDynamics_3(ComponentPose=@n4, Alpha=1, ExternalForce="0,0,0") {
    angularSpringConstant=50
    bAngularSpring=true
    bGravityOverrideInSimSpace=true
    bLinearSpring=true
    linearSpringConstant=200
    simulationSpace=BoneRelative
    alphaBoolBlend {blendOption=Linear}
    boundBone {boneName=dyn_waterDynamics_3}
    componentLinearAccScale=<1.5 1.5 1.15>
    physicsBodyDefinitions [{
      sphereCollisionRadius=0
      boundBone {boneName=dyn_waterDynamics_3}
      boxExtents=<5 5 5>
      constraintSetup {angularLimitsMax=<10 10 10>, angularLimitsMin=<-10 -10 -10>, angularTarget=<1 0 0>, linearAxesMax=<5 0 0>}
      localJointOffset=<0 0 0>
     }]
   }
   n6 = "Transform (Modify) Bone - Bone: pack"(ComponentPose=@n5, Translation="0,-0.8,1", Rotation="0,0,0") {translationMode=BMM_Additive, alphaBoolBlend {blendOption=Linear}, boneToModify {boneName=pack}, translation {y=-0.8, z=1}}
   Output Pose(Result=Component To Local(ComponentPose=@n6)) {layerGroup=None}
   AnimGraphNode_StateMachine_0 = State Machine() {is=title AnimGraphNode_StateMachine}
  ]}
 State Machine {
  entry=Idle
  path=AnimGraph.AnimGraphNode_StateMachine_0.State Machine
  ids=id
  nodes [
   AnimStateNode_0 = Idle()
   AnimStateNode_1 = Hit React() {is=title AnimStateNode}
   AnimStateNode_2 = Dead_Idle() {is=title AnimStateNode}
   AnimStateNode_3 = to_Life() {is=title AnimStateNode}
   AnimStateTransitionNode_0 = AnimStateTransitionNode()
   AnimStateTransitionNode_2 = AnimStateTransitionNode() {bAutomaticRuleBasedOnSequencePlayerInState=true}
   AnimStateTransitionNode_3 = AnimStateTransitionNode() {bAutomaticRuleBasedOnSequencePlayerInState=true}
   AnimStateTransitionNode_4 = AnimStateTransitionNode() {bAutomaticRuleBasedOnSequencePlayerInState=true}
  ]
  links [
   {
    dst=to_Life
    rule=AnimGraph.AnimGraphNode_StateMachine_0.State Machine.AnimStateTransitionNode_3.Transition
    src=Dead_Idle
    via=AnimStateTransitionNode_3
   }
   {
    dst=Dead_Idle
    rule=AnimGraph.AnimGraphNode_StateMachine_0.State Machine.AnimStateTransitionNode_2.Transition
    src=Hit React
    via=AnimStateTransitionNode_2
   }
   {
    dst=Hit React
    rule=AnimGraph.AnimGraphNode_StateMachine_0.State Machine.AnimStateTransitionNode_0.Transition
    src=Idle
    via=AnimStateTransitionNode_0
   }
   {
    dst=Idle
    rule=AnimGraph.AnimGraphNode_StateMachine_0.State Machine.AnimStateTransitionNode_4.Transition
    src=to_Life
    via=AnimStateTransitionNode_4
   }
  ]
 }
 Idle {
  path=AnimGraph.AnimGraphNode_StateMachine_0.State Machine.AnimStateNode_0.Idle
  nodes [Output Animation Pose(Result=Sequence Player 'Teriyaki_Fish_Toon_Idle_Swim_BACKPACK'() {sequence=&Teriyaki_Fish_Toon_Idle_Swim_BACKPACK})]
 }
 Hit React {
  path=AnimGraph.AnimGraphNode_StateMachine_0.State Machine.AnimStateNode_1.Hit React
  nodes [Output Animation Pose(Result=Sequence Player 'Teriyaki_Fish_Toon_to_Death_BACKPACK'() {sequence=/Game/Accessories/FORT_Backpacks/Backpack_M_MED_Teriyaki_Fish_Toon/Animation/Teriyaki_Fish_Toon_to_Death_BACKPACK})]
 }
 Dead_Idle {
  path=AnimGraph.AnimGraphNode_StateMachine_0.State Machine.AnimStateNode_2.Dead_Idle
  nodes [Output Animation Pose(Result=Sequence Player 'Teriyaki_Fish_Toon_Idle_Dead_BACKPACK'() {sequence=/Game/Accessories/FORT_Backpacks/Backpack_M_MED_Teriyaki_Fish_Toon/Animation/Teriyaki_Fish_Toon_Idle_Dead_BACKPACK})]
 }
 to_Life {
  path=AnimGraph.AnimGraphNode_StateMachine_0.State Machine.AnimStateNode_3.to_Life
  nodes [Output Animation Pose(Result=Sequence Player 'Teriyaki_Fish_Toon_to_Life_BACKPACK'() {sequence=/Game/Accessories/FORT_Backpacks/Backpack_M_MED_Teriyaki_Fish_Toon/Animation/Teriyaki_Fish_Toon_to_Life_BACKPACK})]
 }
 Transition {
  path=AnimGraph.AnimGraphNode_StateMachine_0.State Machine.AnimStateTransitionNode_0.Transition
  nodes [Result(bCanEnterTransition=Get Took Damage() -> (Took Damage:bool))]
 }
 Transition {path=AnimGraph.AnimGraphNode_StateMachine_0.State Machine.AnimStateTransitionNode_2.Transition, nodes [Result()]}
 Transition {path=AnimGraph.AnimGraphNode_StateMachine_0.State Machine.AnimStateTransitionNode_3.Transition, nodes [Result()]}
 Transition {path=AnimGraph.AnimGraphNode_StateMachine_0.State Machine.AnimStateTransitionNode_4.Transition, nodes [Result()]}
 EventGraph {nodes [
   Event BlueprintUpdateAnimation()
   > Set jumpReactivity(jumpReactivity:double=SelectFloat(A=0.4, B=0.6, bPickA=Get bIsJumping())) -> (Output_Get:double)
   n1 = Event BlueprintProcessAnimInputEvent()
   > Branch(Condition=OR Boolean(
     A=IsNamed(self=@n1.AnimInputEvent, InEventName=aie.PlayerVitality.ShieldsDecreased)
     B=IsNamed(self=@n1.AnimInputEvent, InEventName=aie.PlayerVitality.HealthDecreased)
    ))
   > Set Took Damage(Took Damage:bool=true) -> (Output_Get:bool)
   Event BlueprintPostEvaluateAnimation() {is=title K2Node_Event}
   > Set Took Damage(Took Damage=false)
   TryGetPawnOwner()
  ]}
]
variables [
 jumpReactivity {type=number}
 Took Damage {type=boolean}
]
